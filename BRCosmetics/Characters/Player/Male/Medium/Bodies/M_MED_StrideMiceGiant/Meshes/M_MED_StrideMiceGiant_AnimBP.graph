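# Animation Blueprint for the StrideMiceGiant player body, layering cosmetic secondary motion onto the attached parent mesh's pose.
# The copied pose runs through a Control Rig, a backpack-attach offset, then chained spring-driven AnimDynamics for belts, torso, neck and zippers.
# The event graph halves most dynamics alphas in front end and rotates the left arm zipper base when DBNO or in front end.
# ArmZipperCompensate is set while targeting but unused in the AnimGraph.
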AnimBlueprint /BRCosmetics/Characters/Player/Male/Medium/Bodies/M_MED_StrideMiceGiant/Meshes/M_MED_StrideMiceGiant_AnimBP
target_skeleton=&/M_MED_StrideMiceGiant_Skeleton
graphs [
 AnimGraph {nodes [
   n1 = Control Rig(Source=Copy Pose From Mesh() {bUseAttachedParent=true}) {
    alphaBoolBlend {blendOption=Linear}
    controlRigAssetReference {blueprintRigClass=&/M_MED_StrideMiceGiant_CtrlRig.M_MED_StrideMiceGiant_CtrlRig_C}
   }
   n2 = "Transform (Modify) Bone - Bone: attach_backpack"!(ComponentPose=Local To Component(LocalPose=@n1), Alpha=1, Translation="0,1.8,-1") -> (Pose) {translationMode=BMM_Additive, alphaBoolBlend {blendOption=Linear}, boneToModify {boneName=attach_backpack}, translation {y=1.8, z=-1}}
   n3 = Anim Dynamics - Bone: Dyn_BeltTorso_L(ComponentPose=@n2, Alpha=1, ExternalForce="0,0,0") {
    angularDampingOverride=1
    angularSpringConstant=20000
    bAngularSpring=true
    bOverrideAngularDamping=true
    bUseGravityOverride=true
    gravityScale=0.35
    alphaBoolBlend {blendOption=Linear}
    boundBone {boneName=Dyn_BeltTorso_L}
    componentAppliedLinearAccClamp=<3000 3000 3000>
    componentLinearAccScale=<0.6 0.6 0.6>
    componentLinearVelScale=<0.6 0.6 0.6>
    gravityOverride {z=-980}
    physicsBodyDefinitions [{
      boundBone {boneName=Dyn_BeltTorso_L}
      boxExtents=<20 30 30>
      constraintSetup {angularLimitsMax=<0 5 1>, angularLimitsMin=<0 -5 -40>, angularTarget=<1 0 0>}
      localJointOffset=<4 0 0>
     }]
    relativeSpaceBone {boneName=root}
   }
   n4 = Anim Dynamics - Bone: Dyn_BeltTorso_R(ComponentPose=@n3, Alpha=1, ExternalForce="0,0,0") {
    angularDampingOverride=1
    angularSpringConstant=20000
    bAngularSpring=true
    bOverrideAngularDamping=true
    bUseGravityOverride=true
    gravityScale=0.35
    alphaBoolBlend {blendOption=Linear}
    boundBone {boneName=Dyn_BeltTorso_R}
    componentAppliedLinearAccClamp=<3000 3000 3000>
    componentLinearAccScale=<0.6 0.6 0.6>
    componentLinearVelScale=<0.6 0.6 0.6>
    gravityOverride {z=-980}
    physicsBodyDefinitions [{
      boundBone {boneName=Dyn_BeltTorso_R}
      boxExtents=<25 30 30>
      constraintSetup {angularLimitsMax=<0 5 1>, angularLimitsMin=<0 -5 -40>, angularTarget=<1 0 0>}
      localJointOffset=<4 0 0>
     }]
    relativeSpaceBone {boneName=root}
   }
   n5 = "Transform (Modify) Bone - Bone: Base_ZiiperArm_L"(ComponentPose=@n4, Alpha=1, Translation="0,0,0", Rotation="0,0,0", Scale="1,1,1") {
    rotationMode=BMM_Additive
    rotationSpace=BCS_BoneSpace
    translationMode=BMM_Additive
    translationSpace=BCS_BoneSpace
    alphaBoolBlend {blendOption=Linear}
    boneToModify {boneName=Base_ZiiperArm_L}
   }
   n6 = Anim Dynamics - Bone: Dyn_Torso_L(ComponentPose=@n5, Alpha=Get Alpha_set() -> (Alpha_set:double), ExternalForce="0,0,0") {
    angularDampingOverride=0.9999
    angularSpringConstant=10000
    bAngularSpring=true
    bOverrideAngularDamping=true
    bUseGravityOverride=true
    gravityScale=0.35
    simulationSpace=RootRelative
    alphaBoolBlend {blendOption=Linear}
    boundBone {boneName=Dyn_Torso_L}
    componentLinearAccScale=<1 1 1>
    componentLinearVelScale=<1 1 1>
    gravityOverride {z=-980}
    physicsBodyDefinitions [{
      boundBone {boneName=Dyn_Torso_L}
      boxExtents=<10 20 25>
      constraintSetup {angularLimitsMax=<0 20 45>, angularLimitsMin=<0 -20 0>, angularTarget=<1 0 0>}
      localJointOffset=<6 0 0>
     }]
    simSpaceSettings {maxAngularAcceleration=800, maxAngularVelocity=100, simSpaceAngularAlpha=1}
   }
   n7 = Anim Dynamics - Bone: Dyn_Torso_R(ComponentPose=@n6, Alpha=Get Alpha_set(), ExternalForce="0,0,0") {
    angularDampingOverride=0.9999
    angularSpringConstant=10000
    bAngularSpring=true
    bOverrideAngularDamping=true
    bUseGravityOverride=true
    gravityScale=0.35
    simulationSpace=RootRelative
    alphaBoolBlend {blendOption=Linear}
    boundBone {boneName=Dyn_Torso_R}
    componentLinearAccScale=<1 1 1>
    componentLinearVelScale=<1 1 1>
    gravityOverride {z=-980}
    physicsBodyDefinitions [{
      boundBone {boneName=Dyn_Torso_R}
      boxExtents=<10 20 25>
      constraintSetup {angularLimitsMax=<0 20 45>, angularLimitsMin=<0 -20 0>, angularTarget=<1 0 0>}
      localJointOffset=<6 0 0>
     }]
    simSpaceSettings {maxAngularAcceleration=800, maxAngularVelocity=100, simSpaceAngularAlpha=1}
   }
   n8 = Anim Dynamics - Bone: Dyn_ZiiperArm_L(ComponentPose=@n7, Alpha=1, ExternalForce="0,0,0") {
    angularDampingOverride=0.99
    angularSpringConstant=1000
    bAngularSpring=true
    bOverrideAngularDamping=true
    bUseGravityOverride=true
    gravityScale=0.35
    simulationSpace=RootRelative
    alphaBoolBlend {blendOption=Linear}
    boundBone {boneName=Dyn_ZiiperArm_L}
    componentAppliedLinearAccClamp=<10000 10000 10000>
    componentLinearAccScale=<1 1 1>
    componentLinearVelScale=<1 1 1>
    gravityOverride {z=-980}
    physicsBodyDefinitions [{
      boundBone {boneName=Dyn_ZiiperArm_L}
      boxExtents=<20 20 20>
      constraintSetup {angularLimitsMax=<0 5 0>, angularLimitsMin=<0 -90 0>, angularTarget=<1 0 0>}
      localJointOffset=<4 0 0>
     }]
    relativeSpaceBone {boneName=root}
   }
   n9 = Anim Dynamics - Bone: Dyn_ZiiperWrist_L(ComponentPose=@n8, Alpha=Get Alpha_set(), ExternalForce="0,0,0") {
    angularDampingOverride=0.9999
    angularSpringConstant=10000
    bAngularSpring=true
    bOverrideAngularDamping=true
    bUseGravityOverride=true
    gravityScale=0.35
    simulationSpace=RootRelative
    alphaBoolBlend {blendOption=Linear}
    boundBone {boneName=Dyn_ZiiperWrist_L}
    componentAppliedLinearAccClamp=<10000 10000 10000>
    componentLinearAccScale=<1 1 1>
    componentLinearVelScale=<1 1 1>
    gravityOverride {z=-980}
    physicsBodyDefinitions [{
      boundBone {boneName=Dyn_ZiiperWrist_L}
      boxExtents=<15 25 30>
      constraintSetup {angularLimitsMax=<0 90 60>, angularLimitsMin=<0 -90 -1>, angularTarget=<1 0 0>}
      localJointOffset=<4 0 0>
     }]
    relativeSpaceBone {boneName=root}
   }
   n10 = Anim Dynamics - Bone: Dyn_ZiiperWrist_R(ComponentPose=@n9, Alpha=Get Alpha_set(), ExternalForce="0,0,0") {
    angularDampingOverride=0.9999
    angularSpringConstant=10000
    bAngularSpring=true
    bOverrideAngularDamping=true
    bUseGravityOverride=true
    gravityScale=0.35
    simulationSpace=RootRelative
    alphaBoolBlend {blendOption=Linear}
    boundBone {boneName=Dyn_ZiiperWrist_R}
    componentAppliedLinearAccClamp=<10000 10000 10000>
    componentLinearAccScale=<1 1 1>
    componentLinearVelScale=<1 1 1>
    gravityOverride {z=-980}
    physicsBodyDefinitions [{
      boundBone {boneName=Dyn_ZiiperWrist_R}
      boxExtents=<15 25 30>
      constraintSetup {angularLimitsMax=<0 90 40>, angularLimitsMin=<0 -90 -1>, angularTarget=<1 0 0>}
      localJointOffset=<4 0 0>
     }]
    relativeSpaceBone {boneName=root}
   }
   n11 = Anim Dynamics - Bone: Dyn_ZipperAnkle_L(ComponentPose=@n10, Alpha=Get Alpha_set(), ExternalForce="0,0,0") {
    angularDampingOverride=0.9999
    angularSpringConstant=2500
    bAngularSpring=true
    bOverrideAngularDamping=true
    bUseGravityOverride=true
    gravityScale=0.35
    simulationSpace=RootRelative
    alphaBoolBlend {blendOption=Linear}
    boundBone {boneName=Dyn_ZipperAnkle_L}
    componentAppliedLinearAccClamp=<3000 3000 3000>
    componentLinearAccScale=<1 1 1>
    componentLinearVelScale=<1 1 1>
    gravityOverride {z=-980}
    physicsBodyDefinitions [{
      boundBone {boneName=Dyn_ZipperAnkle_L}
      boxExtents=<15 25 30>
      constraintSetup {angularLimitsMax=<0 120 5>, angularLimitsMin=<0 0 -5>, angularTarget=<1 0 0>}
      localJointOffset=<4 0 0>
     }]
    relativeSpaceBone {boneName=root}
    simSpaceSettings {maxAngularAcceleration=1000, maxAngularVelocity=1000, simSpaceAngularAlpha=1}
   }
   n12 = Anim Dynamics - Bone: Dyn_ZipperAnkle_R(ComponentPose=@n11, Alpha=Get Alpha_set(), ExternalForce="0,0,0") {
    angularDampingOverride=1
    angularSpringConstant=2500
    bAngularSpring=true
    bOverrideAngularDamping=true
    bUseGravityOverride=true
    gravityScale=0.35
    simulationSpace=BoneRelative
    alphaBoolBlend {blendOption=Linear}
    boundBone {boneName=Dyn_ZipperAnkle_R}
    componentAppliedLinearAccClamp=<3000 3000 3000>
    componentLinearAccScale=<1 1 1>
    componentLinearVelScale=<1 1 1>
    gravityOverride {z=-980}
    physicsBodyDefinitions [{
      boundBone {boneName=Dyn_ZipperAnkle_R}
      boxExtents=<15 25 30>
      constraintSetup {angularLimitsMax=<0 120 5>, angularLimitsMin=<0 0 -5>, angularTarget=<1 0 0>}
      localJointOffset=<5 0 0>
     }]
    relativeSpaceBone {boneName=calf_r}
    simSpaceSettings {maxAngularAcceleration=1000, maxAngularVelocity=1000, simSpaceAngularAlpha=1}
   }
   n13 = Anim Dynamics - Bone: Dyn_ZipperPants_L(ComponentPose=@n12, Alpha=Get Alpha_set(), ExternalForce="0,0,0") {
    angularDampingOverride=1
    angularSpringConstant=10000
    bAngularSpring=true
    bOverrideAngularDamping=true
    bUseGravityOverride=true
    gravityScale=0.35
    simulationSpace=BoneRelative
    alphaBoolBlend {blendOption=Linear}
    boundBone {boneName=Dyn_ZipperPants_L}
    componentAppliedLinearAccClamp=<10000 10000 10000>
    componentLinearAccScale=<1 1 1>
    componentLinearVelScale=<1 1 1>
    gravityOverride {z=-980}
    physicsBodyDefinitions [{
      boundBone {boneName=Dyn_ZipperPants_L}
      boxExtents=<15 25 30>
      constraintSetup {angularLimitsMax=<0 30 60>, angularLimitsMin=<0 -30 -1>, angularTarget=<1 0 0>}
      localJointOffset=<4 0 0>
     }]
    relativeSpaceBone {boneName=calf_l}
    simSpaceSettings {maxAngularAcceleration=1000, maxAngularVelocity=1000, simSpaceAngularAlpha=1}
   }
   n14 = Anim Dynamics - Bone: Dyn_ZipperPants_R(ComponentPose=@n13, Alpha=Get Alpha_set(), ExternalForce="0,0,0") {
    angularDampingOverride=1
    angularSpringConstant=10000
    bAngularSpring=true
    bOverrideAngularDamping=true
    bUseGravityOverride=true
    gravityScale=0.35
    simulationSpace=RootRelative
    alphaBoolBlend {blendOption=Linear}
    boundBone {boneName=Dyn_ZipperPants_R}
    componentAppliedLinearAccClamp=<3000 3000 3000>
    componentLinearAccScale=<1 1 1>
    componentLinearVelScale=<1 1 1>
    gravityOverride {z=-980}
    physicsBodyDefinitions [{
      boundBone {boneName=Dyn_ZipperPants_R}
      boxExtents=<15 25 30>
      constraintSetup {angularLimitsMax=<0 20 60>, angularLimitsMin=<0 -20 -1>, angularTarget=<1 0 0>}
      localJointOffset=<4 0 0>
     }]
    relativeSpaceBone {boneName=root}
    simSpaceSettings {maxAngularAcceleration=1000, maxAngularVelocity=1000, simSpaceAngularAlpha=1}
   }
   n15 = Anim Dynamics - Bone: Dyn_neck(ComponentPose=@n14, Alpha=Get Alpha_set(), ExternalForce="0,0,0") {
    angularDampingOverride=1
    angularSpringConstant=4000
    bAngularSpring=true
    bOverrideAngularDamping=true
    bUseGravityOverride=true
    gravityScale=0.35
    simulationSpace=RootRelative
    alphaBoolBlend {blendOption=Linear}
    boundBone {boneName=Dyn_neck}
    componentAppliedLinearAccClamp=<3000 3000 3000>
    componentLinearAccScale=<1 1 1>
    componentLinearVelScale=<1 1 1>
    gravityOverride {z=-980}
    physicsBodyDefinitions [{
      boundBone {boneName=Dyn_neck}
      boxExtents=<15 25 30>
      constraintSetup {angularLimitsMax=<0 0 25>, angularLimitsMin=<0 -50 -20>, angularTarget=<1 0 0>}
      localJointOffset=<5 0 0>
     }]
    relativeSpaceBone {boneName=root}
    simSpaceSettings {maxAngularAcceleration=1000, maxAngularVelocity=1000, simSpaceAngularAlpha=1}
   }
   n16 = Anim Dynamics - Bone: Dyn_ziipermiddle(ComponentPose=@n15, Alpha=Get Alpha_set(), ExternalForce="0,0,0") {
    angularDampingOverride=1
    angularSpringConstant=10000
    bAngularSpring=true
    bOverrideAngularDamping=true
    bUseGravityOverride=true
    gravityScale=0.35
    simulationSpace=RootRelative
    alphaBoolBlend {blendOption=Linear}
    boundBone {boneName=Dyn_ziipermiddle}
    componentAppliedLinearAccClamp=<3000 3000 3000>
    componentLinearAccScale=<1 1 1>
    componentLinearVelScale=<1 1 1>
    gravityOverride {z=-980}
    physicsBodyDefinitions [{
      boundBone {boneName=Dyn_ziipermiddle}
      boxExtents=<15 20 30>
      constraintSetup {angularLimitsMax=<0 20 60>, angularLimitsMin=<0 -20 -1>, angularTarget=<1 0 0>}
      localJointOffset=<5 0 0>
     }]
    relativeSpaceBone {boneName=spine_04}
    simSpaceSettings {maxAngularAcceleration=1000, maxAngularVelocity=1000, simSpaceAngularAlpha=1}
   }
   n17 = "Transform (Modify) Bone - Bone: Base_ZiiperArm_L"(ComponentPose=@n16, Alpha=1, Translation="0,0,0", Rotation=Get ArmZipperDBNO() -> (ArmZipperDBNO:Rotator), Scale="1,1,1") {rotationMode=BMM_Additive, rotationSpace=BCS_BoneSpace, alphaBoolBlend {blendOption=Linear}, boneToModify {boneName=Base_ZiiperArm_L}}
   Output Pose(Result=Component To Local(ComponentPose=@n17)) {layerGroup=None}
   Get Alpha_set()
  ]}
 EventGraph {nodes [
   Event BlueprintUpdateAnimation()
   > n1 = Sequence!(execute) -> (then_0, then_1, then_2:Exec)
   > Branch(Condition=Get bIsTargeting())
   > Set ArmZipperCompensate(ArmZipperCompensate:Rotator=MakeRotator(Yaw=-4)) -> (Output_Get:Rotator)
   n2 = Branch(execute=@n1.then_1, Condition=OR Boolean(A=Get bIsDBNO(), B=Get bIsInFrontEnd()))
   > Set ArmZipperDBNO(ArmZipperDBNO:Rotator="0,0,-18") -> (Output_Get:Rotator)
   Set ArmZipperDBNO(execute=@n2.else, ArmZipperDBNO="0, 0, 0")
   n3 = Branch(execute=@n1.then_2, Condition=Get bIsInFrontEnd())
   > Set Alpha_set(Alpha_set:double=0.5) -> (Output_Get:double)
   Set Alpha_set(execute=@n3.else, Alpha_set=1)
  ]}
]
variables [
 Alpha_set {type=number}
 ArmZipperCompensate {type=Rotator}
 ArmZipperDBNO {type=Rotator}
]
